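; BGA_Buyable_Item is a Hideouts Market display case that sells an item for currency through player interaction.
; Purchase data drives the displayed mesh, weapon mods and rarity FX, while the server pre-picks loot, destroying cases lacking valid data or disabled by hotfix.
; Interaction checks stock, resources and bans.
; Purchases run on authority: removing currency, spawning pickups targeted at the buyer, sending quest events and updating global and per-player counts.
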(event Custom|TriggerNaughtyListFX
  (params (IsNaughty :type Boolean))
  ;[0.02863 0.0471 0.1753] Naughty List Material Effect
  (Rendering|Material|SetScalarParameterValueonMaterials (Variables|Default|GetDisplayCaseSmall_Wall) "BP_IsBlackListed" (Math|Conversions|ToFloat(Boolean) IsNaughty)))

(event EventBeginPlay
  (bind _ns_market_flash_failed (Variables|Default|GetNS_Market_Flash_Failed))
  ;[0.1177 1 0.3601 0.5] BeginPlay
  (|CallParentFunction)
  (Utilities|FlowControl|Sequence
    (:then_0
      ;;[1 1 1 0.5] Calling this on begin play to hopefully fix weird collision issues
      (CallFunction|CaseMeshSetup) ; Call this on begin play to hopefully fix weird collision issues occurring
    )
    (:then_1
      (EventDispatchers|BindEventtoOnPurchaseDataReady (EventDispatchers|CreateEvent :!event Custom|HandlePurchaseDataReady))
      ;;[1 1 1 0.5] Explicitly set ItemSkeletalMesh's tick disabled. It isn't respecting bStartWithTickEnabled for some reason....FORT-854991
      (Components|Tick|SetComponentTickEnabled (Variables|Default|GetItemSkeletalMesh) false)
      (Class|FortAsyncActionCurrentPlaylistReady|PlaylistDataReadyAsync
        (:OnReady
          (Utilities|IsValid _ns_market_flash_failed
            (:Is%20Valid
              (EventDispatchers|BindEventtoOnSystemFinished _ns_market_flash_failed (EventDispatchers|CreateEvent :!event Custom|OnFlashFailedFinished)))))))
    (:then_2
      (Networking|SwitchHasAuthority
        (:Authority
          (EventDispatchers|BindEventtoOnSelectPurchaseDataFailed (EventDispatchers|CreateEvent :!event Custom|HandlePurchaseDataSelectionFailed)) ; This only broadcasts on the server
        )))))

(event InteractInterface|EventOnInteractFocused
  (params (FortPC :type Fort%20Player%20Controller%20Object%20Reference) (ContextInfo :type Fort%20Interact%20Context%20Info%20Object%20Reference))
  (bind _sm_wallsmall (Variables|Default|GetSM_WallSmall))
  (bind _fxscale (Variables|Default|GetFXScale))
  ;[0.1137 0.2941 1] Highlight Pickup Center Glow element
  ;[0.15 0.1318 0.006687 0.5] Audio Hover Focus and Unfocused and hover loop
  (Audio|PlaySoundatLocation (Variables|Default|GetOnInteractFocusedSound) (Transformation|GetActorLocation) "0, 0, 0" 2.0)
  (Audio|Components|Audio|FadeIn (Variables|Default|GetFocusLoopSound) 0.5 1.0 0.0 "Linear" :!class /Script/Engine.AudioComponent)
  (bind _asb_pickups_default (Utilities|Casting|CastToB_Pickups_Default (Variables|/Script/Engine.ChildActorComponent|GetChildActor (Variables|Default|GetChildActor))))
  (Rendering|Material|SetScalarParameterValueonMaterials (Variables|/Game/Effects/Fort_Effects/Gameplay/Pickups/B_Pickups_Default.B_Pickups_Default_C|GetBG _asb_pickups_default) "HoloIntensity" (* 5.0 _fxscale))
  (Rendering|Material|SetScalarParameterValueonMaterials (Variables|/Game/Effects/Fort_Effects/Gameplay/Pickups/B_Pickups_Default.B_Pickups_Default_C|GetBG _asb_pickups_default) "CameraFacingPush" (* 3e+01 _fxscale))
  (Variables|Default|SetRarityMarket (Variables|/Script/FortniteGame.FortPickupsParent|GetPickupRarityLevel _asb_pickups_default))
  (Rendering|Material|SetColorParameterValueonMaterials _sm_wallsmall "BrightColor" (Utilities|Array|Get(aref) (Variables|/Game/Effects/Fort_Effects/Gameplay/Pickups/B_Pickups_Parent.B_Pickups_Parent_C|GetBackgroundRarityColors _asb_pickups_default) (Variables|/Script/FortniteGame.FortPickupsParent|GetPickupRarityLevel _asb_pickups_default)))
  (Rendering|Material|SetColorParameterValueonMaterials _sm_wallsmall "SideColor" (Utilities|Array|Get(aref) (Variables|/Game/Effects/Fort_Effects/Gameplay/Pickups/B_Pickups_Parent.B_Pickups_Parent_C|GetBackgroundRarityColors _asb_pickups_default) (Variables|/Script/FortniteGame.FortPickupsParent|GetPickupRarityLevel _asb_pickups_default))))

(event Custom|AttachToHideoutManagerNaughtyList
  (params (NaughtyListHideoutManager :type BP%20Hideout%20Naughty%20List%20Manager%20Object%20Reference))
  ;[0.7483 0.1191 0.4928] Populate the Naughty List so that we can track if the player can or not purchase items.
  (Default|BindEventtoNaughtyListUpdated NaughtyListHideoutManager (EventDispatchers|CreateEvent :!event Custom|NaughtyListManagerUpdated))
  (Variables|Default|SetNaughty%20List%20Time (Variables|/DragonPopGameplayContent/Gameplay/HideoutsMarket/BP_Hideout_NaughtyListManager.BP_Hideout_NaughtyListManager_C|GetNaughtyListTime NaughtyListHideoutManager)))

(event Custom|TestNaughty
  (bind _naughty_list (Variables|Default|GetNaughty%20List))
  ;[0.1476 0.1476 0.1476 0.5] Debug function to test naughty list
  (Networking|SwitchHasAuthority
    (:Authority
      (bind _asfort_player_pawn (Utilities|Casting|CastToFortPlayerPawn (Pawn|GetViewTarget (Game|GetPlayerController 0))))
      (Utilities|Array|Add _naughty_list _asfort_player_pawn)
      (Utilities|FlowControl|Delay 5.0)
      (Utilities|Array|RemoveItem _naughty_list _asfort_player_pawn))))

(event InteractInterface|EventOnInteractUnfocused
  (params (FortPC :type Fort%20Player%20Controller%20Object%20Reference) (NewContextInfo :type Fort%20Interact%20Context%20Info%20Object%20Reference))
  ;[0.1137 0.2941 1] Highlight Pickup Center Glow element
  ;[0.15 0.1318 0.006687 0.5] Audio Hover Focus and Unfocused and hover loop
  (Audio|Components|Audio|FadeOut (Variables|Default|GetFocusLoopSound) 0.5 0.0 "Linear" :!class /Script/Engine.AudioComponent)
  (Audio|PlaySoundatLocation (Variables|Default|GetOnInteractUnfocusedSound) (Transformation|GetActorLocation) "0, 0, 0" 2.0)
  (bind _asb_pickups_default (Utilities|Casting|CastToB_Pickups_Default (Variables|/Script/Engine.ChildActorComponent|GetChildActor (Variables|Default|GetChildActor))))
  (Rendering|Material|SetScalarParameterValueonMaterials (Variables|/Game/Effects/Fort_Effects/Gameplay/Pickups/B_Pickups_Default.B_Pickups_Default_C|GetBG _asb_pickups_default) "HoloIntensity" 1.0)
  (Rendering|Material|SetScalarParameterValueonMaterials (Variables|/Game/Effects/Fort_Effects/Gameplay/Pickups/B_Pickups_Default.B_Pickups_Default_C|GetBG _asb_pickups_default) "CameraFacingPush" 0.0)
  (Components|DestroyComponent (Variables|Default|GetFXHighlightGlow)))

(event Custom|NaughtyListManagerUpdated
  (params (NaughtyList :type Array%20of%20Fort%20Player%20Pawn%20Object%20References))
  ;[0.7483 0.1191 0.4928] Populate the Naughty List so that we can track if the player can or not purchase items.
  (Networking|SwitchHasAuthority
    (:Authority
      (Variables|Default|SetNaughty%20List NaughtyList))))

(event EventEndPlay
  (params (EndPlayReason :type EEndPlayReason%20Enum :default "Destroyed"))
  ;[1 0.5237 0.1989 0.5] EndPlay
  (|CallParentFunction EndPlayReason)
  (EventDispatchers|UnbindEventfromOnPurchaseDataReady (EventDispatchers|CreateEvent :!event Custom|HandlePurchaseDataReady))
  (EventDispatchers|UnbindEventfromOnSelectPurchaseDataFailed (EventDispatchers|CreateEvent :!event Custom|HandlePurchaseDataSelectionFailed)))

(event Custom|FeedbackVisibility
  (bind _rarityglow (CallFunction|RarityLightVisibility))
  (Variables|Default|SetRarity%20Glow _rarityglow)
  (bind _newparam (CallFunction|GlowFXRarityVisibility))
  (Variables|Default|SetRarityEffect _newparam))

(event Custom|ApplyFXRarity
  ;[0.1137 0.2941 1] Apply Rarity Effect On Item
  (bind _asb_pickups_default (Utilities|Casting|CastToB_Pickups_Default (Variables|/Script/Engine.ChildActorComponent|GetChildActor (Variables|Default|GetChildActor))))
  (bind _output_get (Variables|/Script/FortniteGame.FortPickupsParent|SetPickupRarityLevel (Math|Conversions|ToInteger(Byte) (Class|FortItemFunctionLibrary|GetRarity (Variables|DisplayItem|GetDisplayItemDef))) _asb_pickups_default))
  (for _array_element (Variables|Default|GetRarityEffect)
    (for _array_element_2 (Variables|Default|GetRarity%20Glow)
      (Rendering|Material|SetScalarParameterValueonMaterials (Variables|Default|GetDisplayCaseGlass_Small) "Rarity_Index" (Math|Conversions|ToFloat(Integer) _output_get))
      (Class|BPickupsDefault|AddRareFX _asb_pickups_default)
      (Class|BPickupsDefault|SetInitialBKGDParam _asb_pickups_default (Class|BTechArtGlobalLibrary|GetDynamicMaterialInstance (Variables|/Game/Effects/Fort_Effects/Gameplay/Pickups/B_Pickups_Default.B_Pickups_Default_C|GetBG _asb_pickups_default)))
      (Transformation|SetRelativeLocation (Variables|/Game/Effects/Fort_Effects/Gameplay/Pickups/B_Pickups_Default.B_Pickups_Default_C|GetBG _asb_pickups_default) (Variables|/Script/Engine.SceneComponent|GetRelativeLocation (Variables|Default|GetItemSkeletalMesh)) false false))))

(event Custom|HandlePurchaseDataReady
  (params (MarketItem :type Building%20Gameplay%20Actor%20Market%20Item%20Object%20Reference))
  (bind _purchasedata (Variables|BuildingGameplayActorMarketItem|GetPurchaseData))
  (bind (_purchaselistrowname _loottiergroup _cost _currency _stockavailableperinstance _maxstockperplayerperinstance _display_itemdef) (Utilities|Struct|BreakMarketGameplayPurchaseRuntimeData _purchasedata))
  (bind _returnvalue (Fortnite|Athena|GameState|GetGameStateAthena))
  (bind _gsc_marketgameplayclass (Variables|PurchaseData|GetGSC_MarketGameplayClass))
  (bind _returnvalue_2 (Actor|GetComponentbyClass _returnvalue _gsc_marketgameplayclass))
  (bind _editoronly_debugtext (Variables|Default|GetEDITORONLY_DebugText))
  ;[0.1813 1 0.2043 0.5] HandlePurchaseDataReady - Called on both Client and Server when purchase data has been filled
  (Variables|Default|SetPurchaseDataReady true)
  (Utilities|FlowControl|Sequence
    (:then_0
      ;;[1 1 1 0.5] TODO - currently the UI and FX need this DisplayItemDef variable set to function,
      ;; they don't support the override icon, rarity, text
      (bind _returnvalue_3 (Utilities|LoadAssetBlocking _display_itemdef)) ; This should already be loaded by the plugin, but just in case
      (bind _asfort_world_item_definition (Utilities|Casting|CastToFortWorldItemDefinition _returnvalue_3))
      (Variables|DisplayItem|SetDisplayItemDef _asfort_world_item_definition)
      (CallFunction|FeedbackVisibility))
    (:then_1
      ;;[1 1 1 0.5] Pick loot drops on the server
      (Networking|SwitchHasAuthority
        (:Authority
          (bind _outloottodrop (Loot|PickLootDrops _loottiergroup 0 -1 "(GameplayTags=)"))
          (Variables|Default|SetLootToDrop _outloottodrop)
          ;;[1 1 1 0.5] Record which items this case has stocked for analytics purposes
          (Utilities|IsValid _returnvalue
            (:Is%20Valid
              (Utilities|IsValid _returnvalue_2
                (:Is%20Valid
                  (Class|FortGameStateComponentMarketGameplay|RecordStockedItemsAnalytics _returnvalue_2 (Variables|PurchaseData|GetMarketGroupTag) (Variables|Default|GetLootToDrop))
                  ;;[0 0 0 0.5] Hotfix Sheet to Destroy Display Cases
                  (bind _returnvalue_4 (Utilities|GetDataTableRow (Variables|Default|GetHotfixSheet) (Variables|PurchaseData|GetPurchaseGroupName)))
                  (if (== (.Action_8_2CDCAF2B42B83FB62D2D3D8FF2D07D5A (Utilities|Struct|BreakHotfixDisplayCases _returnvalue_4)) 1.0)
                    (Actor|DestroyActor)))))))))
    (:then_2
      ;; Debug Text
      (Networking|CanExecuteCosmeticEvents
        (:True
          (Utilities|IsValid _editoronly_debugtext
            (:Is%20Valid
              (Rendering|SetVisibility _editoronly_debugtext (Variables|Debug|GetShowDebugText) false)
              (Rendering|Components|TextRender|SetText _editoronly_debugtext (Utilities|Text|ToText(Name) _loottiergroup)))))))
    (:then_3
      (if (== _currency "/DragonPopGameplayContent/Gameplay/DiamondCurrency/Athena_DiamondItemData.Athena_DiamondItemData")
        (Variables|SetInteractionSpeed_CurveTable (Utilities|Struct|MakeCurveTableRowHandle "/DragonPopGameplayContent/DataTables/DragonPopGameplayContentGameData.DragonPopGameplayContentGameData" (Variables|Default|GetInteraction%20Speed%20Row%20Name)))))))

(event Custom|Succesful%20purchase
  (bind _ns_market_flash_purchase (Variables|Default|GetNS_Market_Flash_Purchase))
  ;[0.2422 0.1496 1] Purchase Feedback
  ; (failed interact feedback is handled in LocalOnFailedInteract)
  (Networking|CanExecuteCosmeticEvents
    (:True
      (Components|Activation|SetActive _ns_market_flash_purchase true true)
      (bind _asb_pickups_default (Utilities|Casting|CastToB_Pickups_Default (Variables|/Script/Engine.ChildActorComponent|GetChildActor (Variables|Default|GetChildActor))))
      (Niagara|SetNiagaraVariable(LinearColor) _ns_market_flash_purchase "RarityColor" (Utilities|Array|Get(aref) (Variables|/Game/Effects/Fort_Effects/Gameplay/Pickups/B_Pickups_Parent.B_Pickups_Parent_C|GetBackgroundRarityColors _asb_pickups_default) (Variables|Default|GetRarityMarket)))
      (Audio|PlaySoundatLocation (Variables|Default|GetBuy%20Start%20Sound) (Transformation|GetActorLocation) "0, 0, 0" 1.0 1.0 0.0 "/Game/Athena/Sounds/Props/Abduction_Pipes/Abduction_Pipe_Leak_Spout_Loop_Att.Abduction_Pipe_Leak_Spout_Loop_Att"))))

(event Custom|HandlePurchaseDataSelectionFailed
  (params (MarketItem :type Building%20Gameplay%20Actor%20Market%20Item%20Object%20Reference))
  ;[1 0.1814 0.09609 0.5] If purchase data CANNOT be selected (eg all items have been HFed off), destroy the case.
  (Networking|SwitchHasAuthority
    (:Authority
      (Fortnite|FortShippingLog self (Utilities|String|Append "HandlePurchaseDataSelectionFailed - Destroying BGA_Buyable_Item as purchase data could not be selected for purchase group " (Utilities|String|ToString(Name) (Variables|PurchaseData|GetPurchaseGroupName))))
      (Actor|DestroyActor))))

(event EventOnDisplayMeshLoaded
  (params (DisplayMeshAsset :type Streamable%20Render%20Asset%20Object%20Reference))
  ;[0.007843 0.1373 1] Display Item Mesh
  (bind _asstatic_mesh (Utilities|Casting|CastToStaticMesh DisplayMeshAsset)
    (:then
      (Components|StaticMesh|SetStaticMesh (Variables|Default|GetItemStaticMesh) _asstatic_mesh)
      (Rendering|SetVisibility (Variables|Default|GetItemSkeletalMesh) false false))
    (:CastFailed
      (bind _asskeletal_mesh (Utilities|Casting|CastToSkeletalMesh DisplayMeshAsset))
      (Components|SkeletalMesh|SetSkeletalMeshAsset (Variables|Default|GetItemSkeletalMesh) _asskeletal_mesh)
      (Rendering|SetVisibility (Variables|Default|GetItemStaticMesh) false false)))
  (CallFunction|SetItemMeshTransform)
  (CallFunction|ApplyFXRarity)
  (CallFunction|LoadModMesh)
  (if (and (Utilities|IsValid (Utilities|ResolveSoftReference (Variables|Default|GetMidasWrap)) :!class /Script/Engine.KismetSystemLibrary) (GameplayTags|HasMatchingGameplayTag (Variables|DisplayItem|GetDisplayItemDef) "(TagName=\"Item.Weapon.Ranged.SMG.Drum.SunRose.Midas\")"))
    (Rendering|Material|SetMaterial (Variables|Default|GetItemSkeletalMesh) 0 (Utilities|ResolveSoftReference (Variables|Default|GetMidasWrap)))))

(event Interaction|EventBlueprintOnInteract
  (params (InteractingPawn :type Fort%20Pawn%20Object%20Reference) (InteractionBeingAttempted :type EInteractionBeingAttempted%20Enum :default "FirstInteraction"))
  ;[0.1294 0 0.149 0.502] OnInteract
  (Variables|Default|SetPlayer%20Buying InteractingPawn)
  (CallFunction|ProcessPurchase)
  (Development|PrintString "Server Interact Purchasing"))

(event Custom|SetItemMeshTransform
  (bind _purchasedata (Variables|BuildingGameplayActorMarketItem|GetPurchaseData))
  (bind (_purchaselistrowname _loottiergroup _cost _currency _stockavailableperinstance _maxstockperplayerperinstance _display_itemdef _display_transform) (Utilities|Struct|BreakMarketGameplayPurchaseRuntimeData _purchasedata))
  (bind _location (.location _display_transform))
  (bind _rotation (.rotation _display_transform))
  (bind _scale (.scale _display_transform))
  ;[0.007843 0.1373 1] Display Item Mesh
  (bind _output_get (Variables|Default|SetMeshBaseTransform (Transformation|GetRelativeTransform (Variables|Default|GetItemSkeletalMesh))))
  (Transformation|SetRelativeTransform (Variables|Default|GetItemSkeletalMesh) (Math|Transform|MakeTransform (+ _location (.location _output_get)) (Math|Rotator|CombineRotators _rotation (.rotation _output_get)) (* _scale (.scale _output_get))) false false)
  (Transformation|SetRelativeTransform (Variables|Default|GetItemStaticMesh) (Math|Transform|MakeTransform (+ _location (.location _output_get)) (Math|Rotator|CombineRotators _rotation (.rotation _output_get)) (* _scale (.scale _output_get))) false false))

(event Interaction|EventBlueprintOnBeginInteract
  (params (InteractingPawn :type Fort%20Pawn%20Object%20Reference) (InteractionBeingAttempted :type EInteractionBeingAttempted%20Enum :default "FirstInteraction") (InteractionType :type TInteractionType%20Enum :default "IT_NoInteraction"))
  ;[0.1294 0 0.149 0.502] OnInteract
  (|CallParentFunction InteractingPawn InteractionBeingAttempted InteractionType)
  (bind _returnvalue (Audio|SpawnSoundatLocation :Sound (Variables|Default|GetInteract%20Sound) :Location (Transformation|GetActorLocation) :VolumeMultiplier 4.0 :PitchMultiplier 1.0 :StartTime 0.0 :bAutoDestroy true))
  (Utilities|FlowControl|Delay 1.2)
  (Audio|Components|Audio|Stop _returnvalue))

(event Custom|ShotGun%20Optic%20Mod%20Mesh
  (bind _modmeshoptic (Variables|Default|GetModMeshOptic))
  (bind _modmeshoptic_2 (Variables|Default|GetModMeshOptic))
  (bind (_relativelocation_x _relativelocation_y _relativelocation_z) (Math|Vector|BreakVector (Variables|/Script/Engine.SceneComponent|GetRelativeLocation_X _modmeshoptic_2)))
  (bind (_relativelocation_x_2 _relativelocation_y_2 _relativelocation_z_2) (Math|Vector|BreakVector (Variables|/Script/Engine.SceneComponent|GetRelativeLocation_X _modmeshoptic)))
  ; Fix Iron Sights ShotGun
  (if (GameplayTags|HasMatchingGameplayTag (Variables|DisplayItem|GetDisplayItemDef) "(TagName=\"Item.Weapon.Ranged.Shotgun.Auto.FirePetal\")")
    (Components|StaticMesh|SetStaticMesh _modmeshoptic_2 (Variables|Default|GetMod%20Mesh))
    (Transformation|AttachComponentToComponent _modmeshoptic_2 (Variables|Default|GetItemSkeletalMesh) "attach_optic" "SnapToTarget" "SnapToTarget" "SnapToTarget" true)
    (Transformation|SetRelativeTransform (Variables|Default|GetModMeshOptic) (Math|Transform|MakeTransform (Math|Vector|MakeVector _relativelocation_x (- _relativelocation_y 1e+01) (- _relativelocation_z 15.0))) false false)
    (elif (GameplayTags|HasMatchingGameplayTag (Variables|DisplayItem|GetDisplayItemDef) "(TagName=\"Item.Weapon.Ranged.Shotgun.Pump.FirePetal\")")
      (Components|StaticMesh|SetStaticMesh _modmeshoptic (Variables|Default|GetMod%20Mesh))
      (Transformation|AttachComponentToComponent _modmeshoptic (Variables|Default|GetItemSkeletalMesh) "attach_optic" "SnapToTarget" "SnapToTarget" "SnapToTarget" true)
      (Transformation|SetRelativeTransform (Variables|Default|GetModMeshOptic) (Math|Transform|MakeTransform (Math|Vector|MakeVector _relativelocation_x_2 (- _relativelocation_y_2 25.0) (- _relativelocation_z_2 1e+01))) false false))))

(event Custom|Load%20Mod%20Mesh
  (bind _modmeshunderbarrel (Variables|Default|GetModMeshUnderbarrel))
  (bind _modmeshbarrel (Variables|Default|GetModMeshBarrel))
  (bind _modmeshiron (Variables|Default|GetModMeshIron))
  (bind _modmeshmagazine (Variables|Default|GetModMeshMagazine))
  (bind _modmeshoptic (Variables|Default|GetModMeshOptic))
  ; Load Mod Mesh
  (bind _asfort_weapon_item_definition (Utilities|Casting|CastToFortWeaponItemDefinition (Variables|DisplayItem|GetDisplayItemDef)))
  (for _array_element (WeaponItemDefinition|GetWeaponModSlots _asfort_weapon_item_definition)
    (Utilities|IsValid (Utilities|Struct|BreakFortWeaponModSlot _array_element)
      (:Is%20Valid
        (bind _returnvalue (Class|FortWeaponModFunctionLibrary|GetModStaticMeshDataforWeapon (Utilities|Struct|BreakFortWeaponModSlot _array_element) _asfort_weapon_item_definition))
        (Variables|Default|SetMod%20Mesh (Utilities|ResolveSoftReference (Utilities|Struct|BreakWeaponModStaticMeshBlueprintData _returnvalue)))
        (switch Utilities|FlowControl|Switch|SwitchonGameplayTag (Variables|/Script/FortniteGame.FortWeaponModItemDefinition|GetModSlot (Utilities|Struct|BreakFortWeaponModSlot _array_element))
          (:WeaponMod.Slot.Underbarrel
            (Components|StaticMesh|SetStaticMesh _modmeshunderbarrel (Utilities|ResolveSoftReference (Utilities|Struct|BreakWeaponModStaticMeshBlueprintData _returnvalue)))
            (Transformation|AttachComponentToComponent _modmeshunderbarrel (Variables|Default|GetItemSkeletalMesh) "attach_under" "SnapToTarget" "SnapToTarget" "SnapToTarget" false))
          (:WeaponMod.Slot.Barrel
            (Components|StaticMesh|SetStaticMesh _modmeshbarrel (Utilities|ResolveSoftReference (Utilities|Struct|BreakWeaponModStaticMeshBlueprintData _returnvalue)))
            (Transformation|AttachComponentToComponent _modmeshbarrel (Variables|Default|GetItemSkeletalMesh) "attach_barrel" "SnapToTarget" "SnapToTarget" "SnapToTarget" false))
          (:WeaponMod.Slot.Optic
            (CallFunction|ShotGunOpticModMesh)
            (CallFunction|AROpticMod)
            (Components|StaticMesh|SetStaticMesh _modmeshoptic (Utilities|ResolveSoftReference (Utilities|Struct|BreakWeaponModStaticMeshBlueprintData _returnvalue)))
            (Transformation|AttachComponentToComponent _modmeshoptic (Variables|Default|GetItemSkeletalMesh) "attach_optic" "SnapToTarget" "SnapToTarget" "SnapToTarget" true)
            (if (GameplayTags|HasMatchingGameplayTag (Variables|DisplayItem|GetDisplayItemDef) "(TagName=\"Item.Weapon.Ranged.Assault.Burst.VividRazor\")")
              (bind _returnvalue_2 (Rendering|Material|CreateDynamicMaterialInstance (Variables|Default|GetItemSkeletalMesh) 1 "/WeaponsUpdated/Gameplay/StrikerBurstAR/Materials/MI_Assault_Burst_Rails.MI_Assault_Burst_Rails" "" :!class /Script/Engine.PrimitiveComponent))
              (Rendering|Material|SetScalarParameterValue _returnvalue_2 (Utilities|Name|MakeLiteralName "Hide Element 10") 0.0 :!class /Script/Engine.MaterialInstanceDynamic)))
          (:WeaponMod.Slot.Magazine
            (Components|StaticMesh|SetStaticMesh _modmeshmagazine (Utilities|ResolveSoftReference (Utilities|Struct|BreakWeaponModStaticMeshBlueprintData _returnvalue)))
            (Transformation|AttachComponentToComponent _modmeshmagazine (Variables|Default|GetItemSkeletalMesh) "attach_mag" "SnapToTarget" "SnapToTarget" "SnapToTarget" false))
          (:WeaponMod.Slot.PistolIrons
            (Components|StaticMesh|SetStaticMesh _modmeshiron (Utilities|ResolveSoftReference (Utilities|Struct|BreakWeaponModStaticMeshBlueprintData _returnvalue)))
            (Transformation|AttachComponentToComponent _modmeshiron (Variables|Default|GetItemSkeletalMesh) "attach_mag" "SnapToTarget" "SnapToTarget" "SnapToTarget" false))))
      (:Is%20Not%20Valid
        (Development|PrintString (Utilities|String|Append "INVALID WEAPON MOD !!! " (Utilities|String|ToString(Text) (Variables|/Script/ItemizationCoreRuntime.ItemDefinitionBase|GetItemName _asfort_weapon_item_definition))))))))

(event Custom|ProcessPurchase
  (bind _returnvalue (Fortnite|Athena|GameState|GetGameStateAthena))
  (bind _gsc_marketgameplayclass (Variables|PurchaseData|GetGSC_MarketGameplayClass))
  (bind _returnvalue_2 (Actor|GetComponentbyClass _returnvalue _gsc_marketgameplayclass))
  ;[1 0.2549 0.7451 0.502] Process Purchase
  ; - Transaction Logic
  ; - Grant Items To Player
  ; - Update purchase counts
  (Networking|SwitchHasAuthority
    (:Authority
      (Utilities|FlowControl|Sequence
        (:then_0
          ;;[0.06274 0.149 0.08627] Remove resources
          (CallFunction|TransactionLogic))
        (:then_1
          ;;[0.149 0.06274 0.0549 0.502] Grant Items
          (CallFunction|GrantItemstoPlayer))
        (:then_2
          ;;[1 1 1 0.5] Send quest purchase event
          (Utilities|IsValid _returnvalue_2
            (:Is%20Valid
              (bind _asfort_player_pawn (Utilities|Casting|CastToFortPlayerPawn (Variables|Default|GetPlayer%20Buying)))
              (Class|FortGameStateComponentMarketGameplay|SendPurchaseEvent _returnvalue_2 _asfort_player_pawn self (Variables|Default|GetLootToDrop) (Variables|BuildingGameplayActorMarketItem|GetPurchaseData)))))
        (:then_3
          ;;[1 1 1 0.5] Update Global Purchase Count
          (Variables|Stock|SetGlobalPurchaseCount (+ (Variables|Stock|GetGlobalPurchaseCount) 1)) ; The OnRep will call OnStockChanged\nand call OnGlobalStockDepleted if applicable
          ;;[1 1 1 0.5] Update Per Player Purchase Count
          (CallFunction|UpdatePlayerPurchaseCount))))))

(event Custom|AR%20Optic%20Mod
  (bind _modmeshoptic (Variables|Default|GetModMeshOptic))
  (bind (_relativelocation_x _relativelocation_y _relativelocation_z) (Math|Vector|BreakVector (Variables|/Script/Engine.SceneComponent|GetRelativeLocation_X _modmeshoptic)))
  (if (GameplayTags|HasMatchingGameplayTag (Variables|DisplayItem|GetDisplayItemDef) "(TagName=\"Item.Weapon.Ranged.Assault.Fast.FirePetal\")")
    (Components|StaticMesh|SetStaticMesh _modmeshoptic (Variables|Default|GetMod%20Mesh))
    (Transformation|AttachComponentToComponent _modmeshoptic (Variables|Default|GetItemSkeletalMesh) "attach_optic" "SnapToTarget" "SnapToTarget" "SnapToTarget" true)
    (Transformation|SetRelativeTransform (Variables|Default|GetModMeshOptic) (Math|Transform|MakeTransform (Math|Vector|MakeVector _relativelocation_x (- _relativelocation_y 15.0) (- _relativelocation_z 15.0))) false false)))

(event Custom|Transaction%20Logic
  (bind _purchasedata (Variables|BuildingGameplayActorMarketItem|GetPurchaseData))
  (bind (_purchaselistrowname _loottiergroup _cost _currency) (Utilities|Struct|BreakMarketGameplayPurchaseRuntimeData _purchasedata))
  ;[1 0.2549 0.7451 0.502] Remove Gold/Diamond from Player's inventory
  (Networking|SwitchHasAuthority
    (:Authority
      (bind _asfort_world_item_definition (Utilities|Casting|CastToFortWorldItemDefinition _currency))
      (Inventory|RemoveItemFromPlayer :PlayerController (Fortnite|GetFortPlayerControllerfromActor (Variables|Default|GetPlayer%20Buying)) :ItemDefinition _asfort_world_item_definition :AmountToRemove _cost :bForceRemoval false :bSkipVerbRemoval false))))

(event Custom|Grant%20Items%20To%20Player
  ;[1 0.2549 0.7451 0.502] Grant Item To Player
  (Networking|SwitchHasAuthority
    (:Authority
      (for _array_element (Variables|Default|GetLootToDrop)
        (bind _asfort_world_item_definition (Utilities|Casting|CastToFortWorldItemDefinition (Class|FortItemFunctionLibrary|BreakFortItemEntry _array_element)))
        (bind _returnvalue (Inventory|SpawnPickupInWorld _asfort_world_item_definition (.Count (Class|FortItemFunctionLibrary|BreakFortItemEntry _array_element)) (Transformation|GetActorLocation) "0.000000,0.000000,0.000000" 0 false true false -1 "Container" "ItemSpawner"))
        (Class|FortPickup|AddSpawnSourceTags _returnvalue (Variables|Default|GetPick%20Up%20Source%20Tag))
        (Pickup|BlueprintSetPickupTarget _returnvalue (Variables|Default|GetPlayer%20Buying) false))
      (CallFunction|SuccesfulPurchase) ; Purchase feedback
    )))

(event Custom|OnFlashFailedFinished
  (params (PSystem :type Niagara%20Particle%20System%20Component%20Object%20Reference))
  (bind _returnvalue (Fortnite|GetFirstLocalFortPlayerController true))
  (bind _self self)
  (bind _interactioncomponent (Variables|/Script/FortniteGame.FortPlayerController|GetInteractionComponent _returnvalue))
  ; On FlashFailed VFX Finished
  (Utilities|IsValid _returnvalue
    (:Is%20Valid
      (if (Inputs|IsUsingInputType "Touch" _returnvalue)
        (if (== _self (Interact|GetInteractActor _interactioncomponent))
          (Class|FortControllerComponentInteraction|SetMobileInteractionHighlight _interactioncomponent true _self))))))

(fn BlueprintCanInteract
  (params (InteractingPawn :type Fort%20Pawn%20Object%20Reference) (InteractionBeingAttempted :type EInteractionBeingAttempted%20Enum :default "FirstInteraction") (InteractionType :type TInteractionType%20Enum :default "IT_NoInteraction"))
  (bind _asfort_player_pawn (Utilities|Casting|CastToFortPlayerPawn InteractingPawn)
    (:then
      (bind _returnvalue (CallFunction|CanPlayerPurchase _asfort_player_pawn))
      (return _returnvalue))
    (:CastFailed
      (return false))))

(fn BlueprintGetAllHighlightableComponents
  (return (Utilities|Array|MakeArray (Variables|Default|GetItemSkeletalMesh) (Variables|Default|GetItemStaticMesh) (Variables|Default|GetModMeshMagazine) (Variables|Default|GetModMeshBarrel) (Variables|Default|GetModMeshUnderbarrel) (Variables|Default|GetModMeshOptic) (Variables|Default|GetModMeshIron)) true))

(fn BlueprintGetFailedInteractionString
  (params (InteractingPawn :type Fort%20Pawn%20Object%20Reference) (InteractionBeingAttempted :type EInteractionBeingAttempted%20Enum :default "FirstInteraction"))
  (bind _asfort_player_pawn (Utilities|Casting|CastToFortPlayerPawn InteractingPawn)
    (:then
      (bind _returnvalue (CallFunction|CanPlayerPurchase _asfort_player_pawn))
      (if (not _returnvalue)
        (return "NSLOCTEXT(\"[33D80652272630864307B3CD606EEBB3]\", \"5C8120814D97B66A68D9E9BC3924BDAA\", \"Error\")")
        (else
          _)))
    (:CastFailed
      _))
  (bind _returnvalue_2 (|CallParentFunction InteractingPawn InteractionBeingAttempted))
  (return _returnvalue_2))

(fn BlueprintUpdateDynamicInteractionOptions
  (params (InteractingPawn :type Fort%20Pawn%20Object%20Reference) (InteractionBeingAttempted :type EInteractionBeingAttempted%20Enum :default "FirstInteraction") (InteractionType :type TInteractionType%20Enum :default "IT_NoInteraction"))
  (bind _outgameplaytags (|CallParentFunction InteractingPawn InteractionBeingAttempted InteractionType))
  (LocalVariables|SetInteraction%20Option _outgameplaytags)
  (if (not (ResourceChecks|DoesPlayerHaveResources InteractingPawn))
    (GameplayTags|AddGameplayTag (LocalVariables|GetInteraction%20Option) "(TagName=\"HUD.Athena.InteractionOptions.DisableInteractKey\")" :!class /Script/GameplayTags.BlueprintGameplayTagLibrary)
    (return (LocalVariables|GetInteraction%20Option))))

(fn CanDisplayEasyInteractButton
  (bind _ns_market_flash_failed (Variables|Default|GetNS_Market_Flash_Failed))
  (Utilities|IsValid _ns_market_flash_failed
    (:Is%20Valid
      (return (not (Components|Activation|IsActive _ns_market_flash_failed))))
    (:Is%20Not%20Valid
      (return true))))

(fn CanPlayerPurchase
  (params (PlayerPawn :type Fort%20Player%20Pawn%20Object%20Reference))
  (bind _personalstockcount (Stock|GetCurrentStock PlayerPawn))
  (return (Math|Boolean|ANDBoolean (!= _personalstockcount 0) (ResourceChecks|DoesPlayerHaveResources PlayerPawn) (not (CallFunction|IsPlayerBanned PlayerPawn)))))

(fn CaseMeshSetup
  (bind _small_glass_display (Variables|DisplayCasesMesh|GetSmall%20Glass%20Display))
  (bind _half_base_small (Variables|DisplayCasesMesh|GetHalf%20Base%20Small))
  (bind _half_glass_large (Variables|DisplayCasesMesh|GetHalf%20Glass%20Large))
  (bind _half_glass_small (Variables|DisplayCasesMesh|GetHalf%20Glass%20Small))
  (bind _small_base_size (Variables|DisplayCasesMesh|GetSmall%20Base%20Size))
  ;[0.007843 0.1373 1] Case Mesh setup
  ;;[1 1 1 0.5] Shelf
  (CallFunction|EnableCaseMeshComponent (Variables|Default|GetDisplayCaseSmall_Wall) (Variables|DisplayCasesMesh|GetSmall%20Wall%20Mesh) false (Variables|DisplayCasesMesh|GetSmall%20Wall%20Mesh))
  (CallFunction|EnableCaseMeshComponent (Variables|Default|GetDisplayCaseLarge_Wall) (Variables|DisplayCasesMesh|GetLarge%20Wall%20Mesh) false (Variables|DisplayCasesMesh|GetLarge%20Wall%20Mesh))
  ;;[1 1 1 0.5] Wall
  (CallFunction|EnableCaseMeshComponent (Variables|Default|GetPlasticCase_Small) (or (Variables|DisplayCasesMesh|GetPlastic%20Wall%20Small) (Variables|DisplayCasesMesh|GetAdd%20Light%20Wall%20Small)) false (or (Variables|DisplayCasesMesh|GetPlastic%20Wall%20Small) (Variables|DisplayCasesMesh|GetAdd%20Light%20Wall%20Small)))
  (CallFunction|EnableCaseMeshComponent (Variables|Default|GetPlasticCase_Large) (or (Variables|DisplayCasesMesh|GetPlastic%20Wall%20Large) (Variables|DisplayCasesMesh|GetAdd%20Light%20Wall%20Large)) false (or (Variables|DisplayCasesMesh|GetPlastic%20Wall%20Large) (Variables|DisplayCasesMesh|GetAdd%20Light%20Wall%20Large)))
  ;;[1 1 1 0.5] Base
  (if (not (Variables|DisplayCasesMesh|GetRemove%20Base%20))
    (CallFunction|EnableCaseMeshComponent (Variables|Default|GetDisplayCaseLarge_Half) (Variables|DisplayCasesMesh|GetHalf%20Base%20Large) false (Variables|DisplayCasesMesh|GetHalf%20Base%20Large))
    (CallFunction|EnableCaseMeshComponent (Variables|Default|GetDisplayCaseSmall_Half) _half_base_small false _half_base_small)
    (if (or (Variables|DisplayCasesMesh|GetHalf%20Base%20Large) _half_base_small)
      (CallFunction|EnableCaseMeshComponent (Variables|Default|GetDisplayCaseBase_Small))
      (CallFunction|EnableCaseMeshComponent (Variables|Default|GetDisplayCaseBase_Large) false true) ; propogate to lightbase children on removal
      (else
        (CallFunction|EnableCaseMeshComponent (Variables|Default|GetDisplayCaseBase_Small) _small_base_size false _small_base_size)
        (CallFunction|EnableCaseMeshComponent (Variables|Default|GetDisplayCaseBase_Large) (not _small_base_size) false (not _small_base_size))))
    ;;[1 1 1 0.5] Glass
    (if (not (Variables|DisplayCasesMesh|GetRemove%20Display%20Glass))
      (CallFunction|EnableCaseMeshComponent (Variables|Default|GetGlass_Half_Large_Flare) _half_glass_large false _half_glass_large)
      (CallFunction|EnableCaseMeshComponent (Variables|Default|GetGlass_Half_Small) _half_glass_small false _half_glass_small)
      (if (not (or _half_glass_large _half_glass_small))
        (CallFunction|EnableCaseMeshComponent (Variables|Default|GetDisplayCaseGlass_Small) _small_glass_display false _small_glass_display)
        (CallFunction|EnableCaseMeshComponent (Variables|Default|GetDisplayCaseGlass_Large) (not _small_glass_display) false (not _small_glass_display))))
    (CallFunction|EnableCaseMeshComponent (Variables|Default|GetDisplayCaseGlass_Small))
    (CallFunction|EnableCaseMeshComponent (Variables|Default|GetDisplayCaseGlass_Large))))

(fn CheckPlayerBoons
  (params (Player%20Interacting :type Fort%20Pawn%20Object%20Reference))
  (bind (_purchaselistrowname _loottiergroup) (Utilities|Struct|BreakMarketGameplayPurchaseRuntimeData (Variables|BuildingGameplayActorMarketItem|GetPurchaseData)))
  (bind (_purchaselistrowname_2 _loottiergroup_2) (Utilities|Struct|BreakMarketGameplayPurchaseRuntimeData (Variables|BuildingGameplayActorMarketItem|GetPurchaseData)))
  (bind (_purchaselistrowname_3 _loottiergroup_3) (Utilities|Struct|BreakMarketGameplayPurchaseRuntimeData (Variables|BuildingGameplayActorMarketItem|GetPurchaseData)))
  (if (and (GameplayTags|HasMatchingGameplayTag Player%20Interacting "(TagName=\"VividRazor.Boons.AdrenalineRush\")") (== _loottiergroup "LTG_Market_Boon_Adrenaline"))
    _
    (elif (and (GameplayTags|HasMatchingGameplayTag Player%20Interacting "(TagName=\"VividRazor.Boons.Pilfer\")") (== _loottiergroup_3 "LTG_Market_Boon_Pilfer"))
      _
      (elif (and (GameplayTags|HasMatchingGameplayTag Player%20Interacting "(TagName=\"VividRazor.Boons.Vulture\")") (== _loottiergroup_2 "LTG_Market_Boon_Vulture"))
        _)))
  (return true))

(fn DoesPlayerHaveResources
  (params (PlayerPawn :type Fort%20Pawn%20Object%20Reference))
  (bind _asfort_player_controller (Utilities|Casting|CastToFortPlayerController (Pawn|GetController PlayerPawn))
    (:then
      (LocalVariables|SetPlayerController _asfort_player_controller)
      (LocalVariables|SetHasResourcesReturn true)
      (bind _output_get (LocalVariables|SetCostData (ResourceChecks|GetCostData)))
      (bind _array_element (Utilities|Array|ForEachLoopwithBreak _output_get)
        (:LoopBody
          (bind _output_get_2 (LocalVariables|SetHasResourcesReturn (>= (Inventory|GetItemQuantityOnPlayer (LocalVariables|GetPlayerController) (Utilities|Struct|BreakSTRUCTHideoutsMarketResourceandCost _array_element)) (.Cost_5_BFDFE411452C5045EB73858C6522FA06 (Utilities|Struct|BreakSTRUCTHideoutsMarketResourceandCost _array_element)))))
          (if _output_get_2
            _
            (else
              _)))
        (:Completed
          _))
      (return (LocalVariables|GetHasResourcesReturn)))
    (:CastFailed
      (return false))))

(fn EnableCaseMeshComponent
  (params (StaticMeshComponent :type Static%20Mesh%20Component%20Object%20Reference) (bVisible :type Boolean) (bPropogateVisibilityToChildren :type Boolean) (bEnableCollision :type Boolean))
  (Utilities|IsValid StaticMeshComponent
    (:Is%20Valid
      (Rendering|SetVisibility StaticMeshComponent bVisible bPropogateVisibilityToChildren)
      (Development|SetHiddeninGame StaticMeshComponent (not bVisible) bPropogateVisibilityToChildren)
      (Collision|SetCollisionEnabled StaticMeshComponent (select bEnableCollision "QueryOnly" "NoCollision")))))

(fn GetOverrideIcon
  (bind (_purchaselistrowname _loottiergroup _cost _currency _stockavailableperinstance _maxstockperplayerperinstance _display_itemdef _display_transform _displayoveride_rarity _displayoverride_icon) (Utilities|Struct|BreakMarketGameplayPurchaseRuntimeData (Variables|BuildingGameplayActorMarketItem|GetPurchaseData)))
  (return _displayoverride_icon))

(fn GetCostData
  (bind (_purchasedata_purchaselistrowname _purchasedata_loottiergroup _purchasedata_cost _purchasedata_currency) (Utilities|Struct|BreakMarketGameplayPurchaseRuntimeData (Variables|GetPurchaseData_PurchaseListRowName)))
  (return (Utilities|Array|MakeArray (Utilities|Struct|MakeSTRUCTHideoutsMarketResourceandCost _purchasedata_currency _purchasedata_cost))))

(fn GetCurrentStock
  (params (PlayerPurchasing :type Fort%20Player%20Pawn%20Object%20Reference))
  (bind _purchasedata (Variables|BuildingGameplayActorMarketItem|GetPurchaseData))
  (bind (_purchaselistrowname _loottiergroup _cost _currency _stockavailableperinstance _maxstockperplayerperinstance) (Utilities|Struct|BreakMarketGameplayPurchaseRuntimeData _purchasedata))
  (bind _playerpurchasing (Variables|GetPlayerPurchasing))
  (bind _playerstate (Variables|/Script/Engine.Pawn|GetPlayerState _playerpurchasing))
  (bind _playerpurchasecount_playerstates (Variables|Stock|GetPlayerPurchaseCount_PlayerStates))
  (bind _returnvalue (Utilities|Array|FindItem _playerpurchasecount_playerstates _playerstate))
  (LocalVariables|SetInfinitePerPlayerStock (== _maxstockperplayerperinstance -1))
  (bind _output_get (LocalVariables|SetInfiniteGlobalStock (== _stockavailableperinstance -1)))
  (LocalVariables|SetGlobalRemainingStock (select _output_get -1 (- _stockavailableperinstance (Variables|Stock|GetGlobalPurchaseCount))))
  ;[1 1 1 0.5] Infinite per player stock, return remaining global stock
  (if (LocalVariables|GetInfinitePerPlayerStock)
    (return (LocalVariables|GetGlobalRemainingStock) (LocalVariables|GetGlobalRemainingStock))
    (else
      (Utilities|FlowControl|Sequence
        (:then_0
          ;[1 1 1 0.5] Calculate remaining stock for the querying player
          (LocalVariables|SetPlayerRemainingStock _maxstockperplayerperinstance)
          (Utilities|IsValid _playerpurchasing
            (:Is%20Valid
              (if (>= _returnvalue 0)
                (LocalVariables|SetPlayerRemainingStock (- (LocalVariables|GetPlayerRemainingStock) (Utilities|Array|Get(aref) (Variables|Stock|GetPlayerPurchaseCount_Count) _returnvalue))) ; Stock remaining for this display case/item for this specific player
              ))))
        (:then_1
          ;[1 1 1 0.5] Return min of global and per player remaining stock
          (bind _can_t_interact (CallFunction|CheckPlayerBoons PlayerPurchasing))
          (return (select _can_t_interact 0 (select (LocalVariables|GetInfiniteGlobalStock) (LocalVariables|GetPlayerRemainingStock) (Math|Integer|Min(Integer) (LocalVariables|GetGlobalRemainingStock) (LocalVariables|GetPlayerRemainingStock)))) (LocalVariables|GetGlobalRemainingStock)))))))

(fn GlowFXRarityVisibility
  (Rendering|SetVisibility (Variables|Default|GetBigTopLight_Flare) (Math|Boolean|ORBoolean false (Variables|DisplayCasesMesh|GetHalf%20Base%20Large) (Rendering|IsVisible (Variables|Default|GetDisplayCaseBase_Large))) false)
  (Rendering|SetVisibility (Variables|Default|GetSmalTopLight_Flare) (or (Variables|DisplayCasesMesh|GetSmall%20Wall%20Mesh) (Variables|DisplayCasesMesh|GetSmall%20Base%20Size)) false)
  (Rendering|SetVisibility (Variables|Default|GetSmalTopLight_Flare_HalfCase) (Variables|DisplayCasesMesh|GetHalf%20Base%20Small) false)
  (Rendering|SetVisibility (Variables|Default|GetBaseLarge_Right) (Rendering|IsVisible (Variables|Default|GetDisplayCaseBase_Large)) false)
  (Rendering|SetVisibility (Variables|Default|GetWallLarge_Left_Flare) (Variables|DisplayCasesMesh|GetLarge%20Wall%20Mesh) false)
  (Rendering|SetVisibility (Variables|Default|GetSmallLight_Flare) (Math|Boolean|ORBoolean (Variables|DisplayCasesMesh|GetSmall%20Wall%20Mesh) (Variables|DisplayCasesMesh|GetHalf%20Base%20Small) (Variables|DisplayCasesMesh|GetSmall%20Base%20Size)) false)
  (Rendering|SetVisibility (Variables|Default|GetHalfLarge_Left_Flare) (or (Variables|DisplayCasesMesh|GetHalf%20Glass%20Large) (Variables|DisplayCasesMesh|GetHalf%20Base%20Large)) false)
  (return (Utilities|Array|MakeArray (Variables|Default|GetWallLarge_Right_Flare) (Variables|Default|GetBaseLarge_Left_Flare) (Variables|Default|GetBaseLarge_Right) (Variables|Default|GetSmalTopLight_Flare_HalfCase) (Variables|Default|GetWallLarge_Left_Flare) (Variables|Default|GetHalfLarge_Right_Flare) (Variables|Default|GetHalfLarge_Left_Flare) (Variables|Default|GetSmallLight_Flare) (Variables|Default|GetBigTopLight_Flare) (Variables|Default|GetSmalTopLight_Flare_HalfCase) (Variables|Default|GetSmalTopLight_Flare) (Variables|Default|GetBigTopLight_Flare_Wall))))

(fn IsPlayerBanned
  (params (PlayerPawn :type Fort%20Player%20Pawn%20Object%20Reference))
  (return (Utilities|Array|ContainsItem (Variables|Default|GetNaughty%20List) PlayerPawn)))

(fn IsStockForPlayerAvailable
  (params (PlayerPawn :type Fort%20Pawn%20Object%20Reference))
  (bind _asfort_player_pawn (Utilities|Casting|CastToFortPlayerPawn PlayerPawn)
    (:then
      (bind _personalstockcount (Stock|GetCurrentStock _asfort_player_pawn))
      (return (!= _personalstockcount 0)))
    (:CastFailed
      (return false))))

(fn LocalOnFailedInteract
  (params (InteractingPawn :type Fort%20Player%20Pawn%20Object%20Reference))
  (bind _fortpawn (LocalVariables|GetFortPawn))
  (bind _returnvalue (Fortnite|GetFortPlayerControllerfromActor _fortpawn))
  (LocalVariables|SetFortPawn InteractingPawn)
  (Audio|PlaySoundatLocation (Variables|Default|GetDepletedCantDoSound) (Transformation|GetActorLocation))
  (Components|Activation|SetActive (Variables|Default|GetNS_Market_Flash_Failed) true true)
  (Utilities|IsValid _fortpawn
    (:Is%20Valid
      (Utilities|IsValid _returnvalue
        (:Is%20Valid
          (if (Inputs|IsUsingInputType "Touch" _returnvalue)
            (Class|FortControllerComponentInteraction|SetMobileInteractionHighlight (Variables|/Script/FortniteGame.FortPlayerController|GetInteractionComponent _returnvalue) false self)))))))

(fn OnGlobalStockDepleted
  (Development|SetHiddeninGame (Variables|Default|GetItemStaticMesh) true false)
  (for _array_element (Variables|Default|GetRarity%20Glow)
    (for _array_element_2 (Variables|Default|GetRarityEffect)
      (Rendering|Material|SetScalarParameterValueonMaterials (Variables|Default|GetDisplayCaseGlass_Small) "UseRarity" 0.0))))

(fn OnRep_GlobalPurchaseCount
  (bind (_purchaselistrowname _loottiergroup _cost _currency _stockavailableperinstance) (Utilities|Struct|BreakMarketGameplayPurchaseRuntimeData (Variables|BuildingGameplayActorMarketItem|GetPurchaseData)))
  (Utilities|FlowControl|Sequence
    (:then_0
      (if (and (Variables|Default|GetPurchaseDataReady) (>= (Variables|Stock|GetGlobalPurchaseCount) _stockavailableperinstance))
        (CallFunction|OnGlobalStockDepleted)))
    (:then_1
      (Default|CallOnStockCountChanged))))

(fn OnRep_NaughtyList
  (Networking|CanExecuteCosmeticEvents
    (:True
      ;[1 1 1 0.5] If any local player controller's pawn is on the naughty list, show naught list FX for this client
      (bind _array_element (Utilities|Array|ForEachLoopwithBreak (Fortnite|GetLocalFortPlayerControllers true))
        (:LoopBody
          (bind _asfort_player_pawn (Utilities|Casting|CastToFortPlayerPawn (Pawn|GetViewTarget _array_element)))
          (LocalVariables|SetFoundNaughty (Utilities|Array|ContainsItem (Variables|Default|GetNaughty%20List) _asfort_player_pawn)))
        (:Completed
          _))
      ;[1 1 1 0.5] Otherwise, make sure naughty list FX are off
      (CallFunction|TriggerNaughtyListFX (LocalVariables|GetFoundNaughty)))))

(fn OnRep_PlayerPurchaseCount_Count
  (Default|CallOnStockCountChanged))

(fn OnRep_PlayerPurchaseCount_PlayerStates
  (Default|CallOnStockCountChanged))

(fn RarityLightVisibility
  (bind _sm_largeglowright (Variables|Default|GetSM_LargeGlowRight))
  (bind _sm_largeglow (Variables|Default|GetSM_LargeGlow))
  (bind _sm_walllargeright (Variables|Default|GetSM_WallLargeRight))
  (bind _sm_walllargelefthalf (Variables|Default|GetSM_WallLargeLeftHalf))
  (bind _sm_wallsmall (Variables|Default|GetSM_WallSmall))
  (Rendering|SetVisibility _sm_largeglow (and (or (Variables|DisplayCasesMesh|GetSmall%20Base%20Size) (Variables|DisplayCasesMesh|GetSmall%20Glass%20Display)) (not (or (Variables|DisplayCasesMesh|GetHalf%20Base%20Large) (Variables|DisplayCasesMesh|GetHalf%20Base%20Small)))) false)
  (Rendering|SetVisibility _sm_largeglowright (Rendering|IsVisible (Variables|Default|GetDisplayCaseBase_Large)) false)
  (Rendering|SetVisibility _sm_wallsmall (or (Variables|DisplayCasesMesh|GetSmall%20Wall%20Mesh) (Variables|DisplayCasesMesh|GetHalf%20Base%20Small)) false)
  (Rendering|SetVisibility _sm_walllargeright (Variables|DisplayCasesMesh|GetLarge%20Wall%20Mesh) false)
  (Rendering|SetVisibility _sm_walllargelefthalf (Variables|DisplayCasesMesh|GetHalf%20Base%20Large) false)
  (return (Utilities|Array|MakeArray _sm_walllargelefthalf (Variables|Default|GetSM_WallLargeRightHalf) _sm_walllargeright (Variables|Default|GetSM_WallLargeLeft) _sm_wallsmall _sm_largeglowright (Variables|Default|GetSM_LargeGlowLeft) _sm_largeglow 0 0 0 0 0 0 0 0 0 0)))

(fn UpdatePlayerPurchaseCount
  (bind _player_buying (Variables|Default|GetPlayer%20Buying))
  (bind _playerstate (Variables|/Script/Engine.Pawn|GetPlayerState _player_buying))
  (bind _playerpurchasecount_playerstates (Variables|Stock|GetPlayerPurchaseCount_PlayerStates))
  (bind _returnvalue (Utilities|Array|FindItem _playerpurchasecount_playerstates _playerstate))
  (if (== _returnvalue -1)
    (Utilities|Array|Add (Variables|Stock|GetPlayerPurchaseCount_PlayerStates) (Variables|/Script/Engine.Pawn|GetPlayerState (Variables|Default|GetPlayer%20Buying)))
    (Utilities|Array|Add (Variables|Stock|GetPlayerPurchaseCount_Count) 1)
    (else
      ;[1 0.7712 0.06661 0.502] HACK
      ; Store off and modify the array in a temp variable, then set it back.
      ; This works around an issue where the array was not replicating when individual values were changed.
      ; This replication issue has been fixed in CL 44396976 (but does not exist in 37.XX branches)
      (bind _output_get (LocalVariables|SetTempCopy (Variables|Stock|GetPlayerPurchaseCount_Count)))
      (Variables|SetBy-RefVar (Utilities|Array|Get(aref) _output_get _returnvalue) (+ (Utilities|Array|Get(aref) _output_get _returnvalue) 1))
      (Variables|Stock|SetPlayerPurchaseCount_Count (LocalVariables|GetTempCopy)))))

(fn ConstructionScript
  (bind _editoronly_debugtext (Variables|Default|GetEDITORONLY_DebugText))
  (|CallParentFunction)
  (Utilities|FlowControl|Sequence
    (:then_0
      (CallFunction|CaseMeshSetup))
    (:then_1
      ; Display PurchaseGroupName on the debug text
      (Networking|CanExecuteCosmeticEvents
        (:True
          (Utilities|IsValid _editoronly_debugtext
            (:Is%20Valid
              (Rendering|SetVisibility _editoronly_debugtext (Variables|Debug|GetShowDebugText) false)
              (Rendering|Components|TextRender|SetText _editoronly_debugtext (Utilities|Text|ToText(Name) (Variables|PurchaseData|GetPurchaseGroupName))))))))))
