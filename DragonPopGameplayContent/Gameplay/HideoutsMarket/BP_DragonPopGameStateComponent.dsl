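; DragonPop game state component that pings a Hideouts Market for players who pick up diamond currency.
; When hotfix-enabled on dedicated servers, it listens for item count changes and, once per player, marks the closest BP_DragonPopIconManager through the player's marker component.
; FilterOutMarketsInStorm narrows candidates to those inside the safe zone by a hotfixable margin.
; A cheat event selects purchase data by group and prints its name, cost and stock.
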(event Custom|Cheat_SelectPurchase
  (params (PurchaseGroupName :type Name))
  (bind _outpurchasedata (CallFunction|SelectPurchaseData "(TagName=\"\")" PurchaseGroupName))
  (Development|PrintString (Utilities|String|Append (Utilities|String|ToString(Name) (Utilities|Struct|BreakMarketGameplayPurchaseRuntimeData _outpurchasedata)) ", " (Utilities|String|ToString(Integer) (.Cost (Utilities|Struct|BreakMarketGameplayPurchaseRuntimeData _outpurchasedata))) ", " (Utilities|String|ToString(Integer) (.StockAvailablePerInstance (Utilities|Struct|BreakMarketGameplayPurchaseRuntimeData _outpurchasedata))))))

(event EventBeginPlay
  (|CallParentFunction)
  (if (ScalableFloat|GetValueasBool (Variables|Default|GetHotfix_PingOnPickUp))
    (if (Networking|IsDedicatedServer)
      (bind (_asynctaskproxy _payload) (GameplayEventRouter|StartListeningToStatefulEvent (GameplayEventRouter|GetGlobalEventRouter) "/Script/FortniteGame.FortItemCountChangedEvent")
        (:OnEventReceived
          (if (== (.ItemDefinition (Utilities|Struct|BreakFortItemCountChangedEvent _payload)) "/DragonPopGameplayContent/Gameplay/DiamondCurrency/Athena_DiamondItemData.Athena_DiamondItemData")
            (bind _outactors (Actor|GetAllActorsOfClass "/DragonPopGameplayContent/Gameplay/HideoutsMarket/BP_DragonPopIconManager.BP_DragonPopIconManager_C"))
            (bind _return (CallFunction|FilterOutMarketsinStorm _outactors))
            (bind _asfort_player_controller (Utilities|Casting|CastToFortPlayerController (Utilities|Struct|BreakFortItemCountChangedEvent _payload)))
            (if (<= (Utilities|Array|FindItem (Variables|Default|GetPlayer%20With%20Dill%20Bits) _asfort_player_controller) -1)
              (Utilities|Array|Add (Variables|Default|GetPlayer%20With%20Dill%20Bits))
              (bind _asbp_dragon_pop_icon_manager (Utilities|Casting|CastToBP_DragonPopIconManager (Fortnite|GetClosestActorfromArray (Game|Player|GetPlayerPawn _asfort_player_controller) _return)))
              (MarkerComponent|MarkActoronServer (Actor|GetComponentbyClass _asfort_player_controller "/Script/FortniteGame.AthenaMarkerComponent") _asbp_dragon_pop_icon_manager false true))))))))

(fn FilterOutMarketsInStorm
  (params (Black%20Market%20Icons :type Array%20of%20Actor%20Object%20References))
  (bind _returnvalue (SafeZone|GetSafeZoneIndicator))
  (Networking|SwitchHasAuthority
    (:Authority
      (Utilities|FlowControl|Sequence
        (:then_0
          (if (>= (ScalableFloat|GetValueatLevel (Variables|Default|GetHotfix_DistanceFromTheStorm_Min)) 0)
            (Utilities|IsValid _returnvalue
              (:Is%20Valid
                (for _array_element Black%20Market%20Icons
                  (Utilities|IsValid _array_element
                    (:Is%20Valid
                      (if (< (Math|Vector|Distance2DSquared(Vector) (+ (Transformation|GetActorLocation _array_element) (* (Math|Vector|Normalize2D(Vector) (Math|Vector|GetUnitDirection(Vector) (SafeZone|GetSafeZoneCenter _returnvalue) (Transformation|GetActorLocation _array_element))) (ScalableFloat|GetValueatLevel (Variables|Default|GetHotfix_DistanceFromTheStorm_Min)))) (SafeZone|GetSafeZoneCenter _returnvalue)) (Math|Float|Square (SafeZone|GetSafeZoneRadius _returnvalue)))
                        (Utilities|Array|Add (LocalVariables|GetBlack%20Market%20Out%20Storm) _array_element))))))
              (:Is%20Not%20Valid
                _)))
          (return Black%20Market%20Icons))
        (:then_1
          (return (LocalVariables|GetBlack%20Market%20Out%20Storm)))))))
