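; Hideout market manager that keeps a timed "naughty list" of pawns who provoke the clerk NPC.
; On authority, once playlist data is ready, it enables a patrol point provider to spawn the clerk and attaches itself to buyable items.
; Damaging the clerk (with a speech bubble) or, optionally, becoming its AI target adds or refreshes an entry.
; A looping timer expires old entries; list changes fire NaughtyListUpdated, and IsOnNaughtyList answers queries.
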(event EventBeginPlay
  (bind _patrolpointprovider (Variables|Default|GetPatrolPointProvider))
  ; Spawn the Clerk
  (Networking|SwitchHasAuthority
    (:Authority
      (Class|FortAsyncActionCurrentPlaylistReady|PlaylistDataReadyAsync
        (:OnReady
          (Utilities|FlowControl|Delay 0.1)
          (Utilities|IsValid _patrolpointprovider
            (:Is%20Valid
              (EventDispatchers|BindEventtoOnActorSpawnedResult _patrolpointprovider (EventDispatchers|CreateEvent :!event Custom|OnDeskClerkSpawned))
              (Class|FortAthenaLivingWorldPointProviderInterface|EnablePointProvider_DEPRECATED _patrolpointprovider)
              (for _array_element (Variables|Default|GetHideoutItemsForSale)
                (Utilities|IsValid _array_element
                  (:Is%20Valid
                    (Class|BGABuyableItem|AttachtoHideoutManagerNaughtyList _array_element self))))
              (bind _returnvalue (Utilities|Time|SetTimerbyEvent (EventDispatchers|CreateEvent :!event Custom|ValidateNaughtyListStillValid) (ScalableFloat|GetValueatLevel (Variables|Default|GetValidateNaughtyListTimer)) true))
              (Variables|InternalData|SetHNDL_ValidateNaughtyListTimer _returnvalue))))))))

(event Custom|HandleSpeechComponent
  ; Handle Speech Component - Tries to Add a Speech Component and when Damaged, Display Speech Component
  (Utilities|IsValid (Actor|GetComponentbyClass (Variables|InternalData|GetClerkPawn) "/Script/FortniteGame.FortPawnComponent_SpeechBubble")
    (:Is%20Valid
      (Fortnite|FortShippingLog self "Trying to add a Fort Speech Component to Clerk NPC, but Clerk already has a Fort Speech Component!" true))
    (:Is%20Not%20Valid
      (Game|AddComponentbyClass (Variables|InternalData|GetClerkPawn) (Variables|Default|GetSpeechComponent))))
  (EventDispatchers|BindEventtoOnDamaged (Variables|InternalData|GetClerkPawn) (EventDispatchers|CreateEvent :!event Custom|ClerkWasDamaged)))

(event Custom|OnDeskClerkSpawned
  (params (SpawnedActor :type Actor%20Object%20Reference) (bResult :type Boolean))
  ; Bind on when the AI has a target - only happens when he's "Hostile"
  (bind _asbp_player_pawn_non_participant (Utilities|Casting|CastToBP_PlayerPawn_NonParticipant SpawnedActor))
  (bind _output_get (Variables|InternalData|SetClerkPawn _asbp_player_pawn_non_participant))
  (if (ScalableFloat|GetValueasBool (Variables|Default|GetCanClerkTargetsBeBanned))
    (bind _asfort_athena_aibot_controller (Utilities|Casting|CastToFortAthenaAIBotController (AI|GetAIController _output_get)))
    (Perception|BindEventtoOnCurrentTargetChangedEvent _asfort_athena_aibot_controller (EventDispatchers|CreateEvent :!event Custom|AITargetChanged))
    (else
      _))
  (CallFunction|HandleSpeechComponent))

(event Custom|ClerkWasDamaged
  (params (DamagedActor :type Actor%20Object%20Reference) (Damage :type Float%20%28single-precision%29) (InstigatedBy :type Controller%20Object%20Reference) (DamageCauser :type Actor%20Object%20Reference) (HitLocation :type Vector) (FHitComponent :type Primitive%20Component%20Object%20Reference) (BoneName :type Name) (Momentum :type Vector))
  ; Handle Speech Component - Tries to Add a Speech Component and when Damaged, Display Speech Component
  (bind _asfort_player_controller_athena (Utilities|Casting|CastToFortPlayerControllerAthena InstigatedBy)
    (:then
      (CallFunction|AddActortoNaughtyList (Game|Player|GetPlayerPawn _asfort_player_controller_athena)))
    (:CastFailed
      (bind _asfort_athena_aibot_controller (Utilities|Casting|CastToFortAthenaAIBotController InstigatedBy))
      (bind _asfort_athena_aibot_controller_2 (Utilities|Casting|CastToFortAthenaAIBotController (Pawn|GetController (Variables|InternalData|GetClerkPawn))))
      (if (!= _asfort_athena_aibot_controller _asfort_athena_aibot_controller_2)
        (CallFunction|AddActortoNaughtyList (Pawn|GetControlledPawn _asfort_athena_aibot_controller)))))
  (Class|BPFLAICommonFunctions|PlaySpeechBubble (Variables|InternalData|GetClerkPawn) "(TagName=\"Athena.AI.SpeechBubbles.DragonPop.ClerkDamaged\")" false true))

(event Custom|AddActorToNaughtyList
  (params (Actor :type Actor%20Object%20Reference))
  (bind _naughtylist (Variables|Default|GetNaughtyList))
  (bind _naughtylistwithtime (Variables|Default|GetNaughtyListWithTime))
  ; Add Targets to the Naughty List
  (bind _asfort_player_pawn (Utilities|Casting|CastToFortPlayerPawn Actor))
  (if (> (Utilities|Array|FindItem _naughtylist _asfort_player_pawn) 0)
    (Variables|InternalData|SetIndexOfNaughtyList -1)
    (for (_array_element _array_index) _naughtylistwithtime
      (if (== (Utilities|Struct|BreakSTRUCTNaughtyList _array_element) _asfort_player_pawn)
        (Variables|InternalData|SetIndexOfNaughtyList _array_index)))
    (if (> (Variables|InternalData|GetIndexOfNaughtyList) -1)
      (Utilities|Array|SetArrayElem _naughtylistwithtime (Variables|InternalData|GetIndexOfNaughtyList) (Utilities|Struct|MakeSTRUCTNaughtyList _asfort_player_pawn (Utilities|Time|GetGameTimeinSeconds)) false))
    (else
      (Utilities|Array|AddUnique _naughtylist _asfort_player_pawn)
      (Utilities|Array|AddUnique (Variables|Default|GetNaughtyListWithTime) (Utilities|Struct|MakeSTRUCTNaughtyList _asfort_player_pawn (Utilities|Time|GetGameTimeinSeconds)))
      (Default|CallNaughtyListUpdated (Variables|Default|GetNaughtyList)))))

(event Custom|AITargetChanged
  (params (OldTarget :type Actor%20Object%20Reference) (NewTarget :type Actor%20Object%20Reference))
  ; Add Targets to the Naughty List
  (CallFunction|AddActortoNaughtyList NewTarget))

(event Custom|ValidateNaughtyListStillValid
  ; Check if the Player should not be on the Naughty List anymore
  (Variables|Default|SetItemRemovedFromList false)
  (bind _output_get (Variables|InternalData|SetNaughtyListWithTimeInternal (Variables|Default|GetNaughtyListWithTime)))
  (for _array_element _output_get
    (if (>= (- (Utilities|Time|GetGameTimeinSeconds) (.TimeSinceWasNaughty_5_8FCAFEFA4CBEE0B3227E3FBA77406F30 (Utilities|Struct|BreakSTRUCTNaughtyList _array_element))) (ScalableFloat|GetValueatLevel (Variables|Default|GetNaughtyListTime)))
      (Utilities|Array|RemoveItem (Variables|Default|GetNaughtyList) (Utilities|Struct|BreakSTRUCTNaughtyList _array_element))
      (Utilities|Array|RemoveItem (Variables|Default|GetNaughtyListWithTime) _array_element)
      (Variables|Default|SetItemRemovedFromList true)))
  (if (Variables|Default|GetItemRemovedFromList)
    (Default|CallNaughtyListUpdated (Variables|Default|GetNaughtyList))))

(event EventEndPlay
  (params (EndPlayReason :type EEndPlayReason%20Enum :default "Destroyed"))
  (Utilities|Time|ClearandInvalidateTimerbyHandle (Variables|InternalData|GetHNDL_ValidateNaughtyListTimer)))

(fn IsOnNaughtyList
  (params (Actor :type Actor%20Object%20Reference))
  (bind _asfort_player_pawn (Utilities|Casting|CastToFortPlayerPawn Actor))
  (return (Utilities|Array|ContainsItem (Variables|Default|GetNaughtyList) _asfort_player_pawn)))

(fn ConstructionScript)
